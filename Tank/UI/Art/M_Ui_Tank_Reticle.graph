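# Additive UI material that colors the tank's aiming reticle from a single mask texture.
# The texture's red channel blends from black to amber.
# The green channel then blends from a solid red-orange to that result, so one texture drives two tints.
# Under additive blending, black areas add nothing to the screen.
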Material /Tank/UI/Art/M_Ui_Tank_Reticle
graph {
 properties {blendMode=BLEND_Additive, materialDomain=MD_UI, pixelDepthOffsetMode=PDOM_Legacy, usedShadingModels=Default Lit}
 nodes [
  n1 = TextureSample() {texture=&/T_Ui_Tank_Reticle}
  LinearInterpolate(
   A=Constant4Vector() {constant=<rgba 1 0.112 0 1>}.RGBA
   B=LinearInterpolate(A=Constant(), B=Constant4Vector() {constant=<rgba 1 0.5029 0 1>}.RGBA, Alpha=@n1.R)
   Alpha=@n1.G
  )
 ]
}
