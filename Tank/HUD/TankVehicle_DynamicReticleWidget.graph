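# Tank HUD reticle widget that shows cannon reload progress in a small centered indicator over the reticle image.
# OnCannonFired plays the one-second Reload animation at 1/FireTime speed, animating the indicator material's fill percentage and alpha over the fire time.
# When Reload finishes, a bound event plays Burst, a brief expanding, fading ring.
# Construct starts Reload at its end so the indicator begins full.
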WidgetBlueprint /Tank/HUD/TankVehicle_DynamicReticleWidget
animations [
 Reload {
  binding_count=1
  length_seconds=1
  object=&self:Reload
  sequence {
   bindings [Image_42 {
     id=#1
     possessed_class=Image
     tracks [{
       type=MovieSceneWidgetMaterialTrack
       sections [{
         type=MovieSceneParameterSection
         channels [
          Fill_Percentage {type=MovieSceneScriptingFloatChannel, keys [[0,4],[60,100]]}
          Fill_Alpha {type=MovieSceneScriptingFloatChannel, keys [[0,0.5],[60,1]]}
         ]
        }]
      }]
    }]
   display_rate {numerator=60}
  }
 }
 Burst {
  binding_count=1
  length_seconds=0.25
  object=&self:Burst
  sequence {
   bindings [Image_Burst {
     id=#2
     possessed_class=Image
     tracks [
      {
       type=MovieScene2DTransformTrack
       sections [{
         type=MovieScene2DTransformSection
         channels [
          Translation.X {type=MovieSceneScriptingFloatChannel}
          Translation.Y {type=MovieSceneScriptingFloatChannel}
          Angle {type=MovieSceneScriptingFloatChannel}
          Scale.X {type=MovieSceneScriptingFloatChannel, keys [[0,1],[15,2.5]]}
          Scale.Y {type=MovieSceneScriptingFloatChannel, keys [[0,1],[15,2.5]]}
          Shear.X {type=MovieSceneScriptingFloatChannel}
          Shear.Y {type=MovieSceneScriptingFloatChannel}
         ]
        }]
      }
      {
       type=MovieSceneFloatTrack
       display_name=Render Opacity
       sections [{type=MovieSceneFloatSection, channels [None {type=MovieSceneScriptingFloatChannel, keys [[0,0],[2,1],[15,0]]}]}]
      }
      {
       type=MovieSceneWidgetMaterialTrack
       sections [{type=MovieSceneParameterSection, channels [RingWidth {type=MovieSceneScriptingFloatChannel, keys [[0,9],[5,14.4854],[15,2]]}]}]
      }
     ]
    }]
   display_rate {numerator=60}
  }
 }
]
graphs [EventGraph {nodes [
   n1 = OnCannonFired!() -> (OutputDelegate, then, FireTime:double=~) {is=title K2Node_CustomEvent}
   n2 = Get Reload() -> (Reload:Widget Animation Object Reference)
   > n3 = PlayAnimationForward(InAnimation=@n2, PlaybackSpeed=float / float(A=1, B=@n1.FireTime))
   n4 = reload finished() {is=title K2Node_CustomEvent}
   BindToAnimationFinished(execute=@n3.then, Animation=@n2, Delegate=@n4)
   Event Construct()
   > PlayAnimation(InAnimation=Get Reload(), StartAtTime=1)
   PlayAnimationForward(execute=@n4.then, InAnimation=Get Burst() -> (Burst:Widget Animation Object Reference))
  ]}]
variables []
widgets {
 info {inheritedWidgetCount=0, parentClass=&UserWidget, rootWidgetClass=&SizeBox, widgetCount=6}
 widgets [
  SizeBox_0 {
   widgetClassPath=&SizeBox
   widget=SizeBox_0
   properties {bOverride_HeightOverride=true, bOverride_WidthOverride=true, heightOverride=128, widthOverride=128}
  }
  Overlay_0 {widgetClassPath=&Overlay, parent=SizeBox_0, widget=Overlay_0, slot {class=SizeBoxSlot}}
  ReticleImage {
   widgetClassPath=&Image
   bIsVariable=true
   parent=Overlay_0
   widget=ReticleImage
   properties {brush {
     bIsDynamicallyLoaded=false
     drawAs=Image
     imageType=NoImage
     mirroring=NoMirror
     resourceName=None
     resourceObject=/Tank/UI/Art/M_Ui_Tank_Reticle
     tiling=NoTile
     imageSize=<108 108>
     margin=<ltrb 0 0 1 1>
     outlineSettings {
      bUseBrushTransparency=false
      roundingType=HalfHeightRadius
      width=0
      color {colorUseRule=UseColor_Specified, specifiedColor=<rgba 0 0 0 0>}
      cornerRadii=<0 0 0 0>
     }
     tintColor {colorUseRule=UseColor_Specified, specifiedColor=<rgba 1 1 1 1>}
     uVRegion {bIsValid=false, max=<0 0>, min=<0 0>}
    }}
   slot {class=OverlaySlot, properties {horizontalAlignment=HAlign_Center, verticalAlignment=VAlign_Center}}
  }
  Image {
   widgetClassPath=&Image
   bIsVariable=true
   parent=Overlay_0
   widget=Image
   properties {
    brush {
     bIsDynamicallyLoaded=false
     drawAs=Image
     imageType=NoImage
     mirroring=NoMirror
     resourceName=None
     resourceObject=/Tank/UI/Art/M_UI_Tank_Reticle_Indicator_Backing
     tiling=NoTile
     imageSize=<26 26>
     margin=<ltrb 0 0 0 0>
     outlineSettings {
      bUseBrushTransparency=false
      roundingType=HalfHeightRadius
      width=0
      color {colorUseRule=UseColor_Specified, specifiedColor=<rgba 0 0 0 0>}
      cornerRadii=<0 0 0 0>
     }
     tintColor {colorUseRule=UseColor_Specified, specifiedColor=<rgba 1 1 1 1>}
     uVRegion {bIsValid=false, max=<0 0>, min=<0 0>}
    }
    colorAndOpacity=<rgba 0 0 0 0.25>
   }
   slot {class=OverlaySlot, properties {horizontalAlignment=HAlign_Center, verticalAlignment=VAlign_Center}}
  }
  Image_42 {
   widgetClassPath=&Image
   bIsVariable=true
   parent=Overlay_0
   widget=Image_42
   properties {brush {
     bIsDynamicallyLoaded=false
     drawAs=Image
     imageType=NoImage
     mirroring=NoMirror
     resourceName=None
     resourceObject=/Tank/UI/Art/M_UI_Tank_Reticle_Indicator_Base
     tiling=NoTile
     imageSize=<24 24>
     margin=<ltrb 0 0 0 0>
     outlineSettings {
      bUseBrushTransparency=false
      roundingType=HalfHeightRadius
      width=0
      color {colorUseRule=UseColor_Specified, specifiedColor=<rgba 0 0 0 0>}
      cornerRadii=<0 0 0 0>
     }
     tintColor {colorUseRule=UseColor_Specified, specifiedColor=<rgba 1 1 1 1>}
     uVRegion {bIsValid=false, max=<0 0>, min=<0 0>}
    }}
   slot {class=OverlaySlot, properties {horizontalAlignment=HAlign_Center, verticalAlignment=VAlign_Center}}
  }
  Image_Burst {
   widgetClassPath=&Image
   bIsVariable=true
   parent=Overlay_0
   widget=Image_Burst
   properties {
    renderOpacity=0
    brush {
     bIsDynamicallyLoaded=false
     drawAs=Image
     imageType=NoImage
     mirroring=NoMirror
     resourceName=None
     resourceObject=&M_UI_Tank_Reticle_Indicator_Base
     tiling=NoTile
     imageSize=<24 24>
     margin=<ltrb 0 0 0 0>
     outlineSettings {
      bUseBrushTransparency=false
      roundingType=HalfHeightRadius
      width=0
      color {colorUseRule=UseColor_Specified, specifiedColor=<rgba 0 0 0 0>}
      cornerRadii=<0 0 0 0>
     }
     tintColor {colorUseRule=UseColor_Specified, specifiedColor=<rgba 1 1 1 1>}
     uVRegion {bIsValid=false, max=<0 0>, min=<0 0>}
    }
    renderTransform {angle=0, scale=<1.5 1.5>, shear=<0 0>, translation=<0 0>}
   }
   slot {class=OverlaySlot, properties {horizontalAlignment=HAlign_Center, verticalAlignment=VAlign_Center}}
  }
 ]
}
